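# Sound cue for the exterior rocking chair ambient spot at the ACabin POI, playing a finite run of randomized rocking sounds.
# A random node picks one of two loop lengths, each repeatedly drawing from two nested random groups of rocking chair waves.
# All quality levels share this graph.
# Playback uses distant attenuation, the Fort_Doors sound class and a reverb submix send, with concurrency blocking any second instance.
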SoundCue /Game/Sounds/Artemis/POI/ACabin/Spots/Spot_ACabin_AmbientLoop_Exterior_Rocking_Chair_Cue
root=QualityLevel_0
nodes [
 {id=Looping_0, children [Random_1], properties {bLoopIndefinitely=false, loopCount=16}}
 {id=Looping_1, children [Random_1], properties {bLoopIndefinitely=false, loopCount=10}}
 {id=QualityLevel_0, children [Random_3,Random_3,Random_3,Random_3]}
 {id=Random_0, children [WavePlayer_1,WavePlayer_3,WavePlayer_4,WavePlayer_5], properties {weights [1,1,1,1]}}
 {id=Random_1, children [Random_0,Random_2], properties {weights [1,1]}}
 {id=Random_2, children [WavePlayer_11, WavePlayer_12, WavePlayer_13, WavePlayer_14, WavePlayer_15], properties {weights [1,1,1,1,1]}}
 {id=Random_3, children [Looping_0,Looping_1], properties {weights [1,1]}}
 {id=WavePlayer_1, properties {soundWaveAssetPtr=&/Spot_ACabin_RockingChair_10}}
 {id=WavePlayer_11, properties {soundWaveAssetPtr=&/Spot_ACabin_RockingChair_05}}
 {id=WavePlayer_12, properties {soundWaveAssetPtr=&/Spot_ACabin_RockingChair_06}}
 {id=WavePlayer_13, properties {soundWaveAssetPtr=&/Spot_ACabin_RockingChair_07}}
 {id=WavePlayer_14, properties {soundWaveAssetPtr=&/Spot_ACabin_RockingChair_08}}
 {id=WavePlayer_15, properties {soundWaveAssetPtr=&/Spot_ACabin_RockingChair_09}}
 {id=WavePlayer_3, properties {soundWaveAssetPtr=&/Spot_ACabin_RockingChair_02}}
 {id=WavePlayer_4, properties {soundWaveAssetPtr=&/Spot_ACabin_RockingChair_03}}
 {id=WavePlayer_5, properties {soundWaveAssetPtr=&/Spot_ACabin_RockingChair_04}}
]
properties {
 attenuationSettings=&/Spot_ACabin_RockingChair_Distant_Att
 bEnableBusSends=false
 bOverrideConcurrency=true
 maxDistance=7500
 soundClassObject=&Fort_Doors
 volumeMultiplier=2.5
 concurrencyOverrides {maxCount=1, resolutionRule=PreventNew}
 preEffectBusSends [{}]
 soundSubmixSends [{maxSendLevel=1.3, minSendLevel=1, soundSubmix=&Fortnite_Reverb_Default}]
}
