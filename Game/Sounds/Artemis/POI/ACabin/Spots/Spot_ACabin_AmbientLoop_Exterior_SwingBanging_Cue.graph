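# Sound cue for a distant swinging, banging exterior ambient spot at the ACabin POI, using metal-bang attenuation and the Fort_Doors sound class.
# A Random node picks one of two equally weighted volume envelopes, one fading out by 21 seconds and the other by 15.
# Each runs through its own default Modulator into a shared Quality Level node, with every tier using the same wave.
# Concurrency allows one instance and blocks new ones.
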SoundCue /Game/Sounds/Artemis/POI/ACabin/Spots/Spot_ACabin_AmbientLoop_Exterior_SwingBanging_Cue
root=Random_0
nodes [
 {
  id=Enveloper_0
  children [Modulator_0]
  properties {
   pitchCurve {editorCurveData {keys [{time=0}]}}
   volumeCurve {editorCurveData {keys [<at 0 1>, {interpMode=RCIM_Cubic, leaveTangent=-0.05, time=15, value=1}, <at 21 0>]}}
  }
 }
 {
  id=Enveloper_1
  children [Modulator_1]
  properties {
   pitchCurve {editorCurveData {keys [{time=0}]}}
   volumeCurve {editorCurveData {keys [<at 0 1>, {interpMode=RCIM_Cubic, leaveTangent=-0.05, time=10, value=1}, <at 15 0>]}}
  }
 }
 {id=Modulator_0, children [QualityLevel_0]}
 {id=Modulator_1, children [QualityLevel_0]}
 {id=QualityLevel_0, children [WavePlayer_0,WavePlayer_0,WavePlayer_0,WavePlayer_0]}
 {id=Random_0, children [Enveloper_0,Enveloper_1], properties {weights [1,1]}}
 {id=WavePlayer_0, properties {soundWaveAssetPtr=&/Spot_ACabin_AmbientLoop_Exterior_SwingBanging_01}}
]
properties {
 attenuationSettings=&/Spot_ACabin_Metal_Bang_Distant_Att
 bEnableBusSends=false
 bOverrideConcurrency=true
 maxDistance=8500
 soundClassObject=&Fort_Doors
 volumeMultiplier=2
 concurrencyOverrides {maxCount=1, resolutionRule=PreventNew}
 preEffectBusSends [{}]
}
