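# Sound cue for a bench bang spot sound at the Artemis ACabin point of interest.
# All four quality-level inputs share one Random node, which picks among six equally weighted bang wave variations.
# The cue boosts volume and overrides attenuation with listener focus, set quieter in focus than out of it.
# It is assigned to the Fort_Vehicles_Interior sound class, despite being a cabin spot sound.
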SoundCue /Game/Sounds/Artemis/POI/ACabin/Spots/Spot_ACabin_Bench_Bang_Cue
root=QualityLevel_0
nodes [
 {id=QualityLevel_0, children [Random_1,Random_1,Random_1,Random_1]}
 {id=Random_1, children [WavePlayer_3,WavePlayer_4,WavePlayer_5,WavePlayer_6,WavePlayer_7,WavePlayer_8], properties {weights [1,1,1,1,1,1]}}
 {id=WavePlayer_3, properties {soundWaveAssetPtr=&/Spot_ACabin_Bench_Bang_01}}
 {id=WavePlayer_4, properties {soundWaveAssetPtr=&/Spot_ACabin_Bench_Bang_02}}
 {id=WavePlayer_5, properties {soundWaveAssetPtr=&/Spot_ACabin_Bench_Bang_03}}
 {id=WavePlayer_6, properties {soundWaveAssetPtr=&/Spot_ACabin_Bench_Bang_05}}
 {id=WavePlayer_7, properties {soundWaveAssetPtr=&/Spot_ACabin_Bench_Bang_06}}
 {id=WavePlayer_8, properties {soundWaveAssetPtr=&/Spot_ACabin_Bench_Bang_07}}
]
properties {
 bEnableBusSends=false
 bOverrideAttenuation=true
 maxDistance=1750
 soundClassObject=&Fort_Vehicles_Interior
 volumeMultiplier=2
 attenuationOverrides {bEnableListenerFocus=true, falloffDistance=1000, focusVolumeAttenuation=0.75, nonFocusVolumeAttenuation=1.2, attenuationShapeExtents {x=750}}
 preEffectBusSends [{}]
}
