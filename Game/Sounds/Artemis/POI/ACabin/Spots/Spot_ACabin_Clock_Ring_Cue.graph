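# Sound cue for a clock ring spot sound at the ACabin point of interest, classed as outdoor ambience.
# A quality-level root routes all four of its branches to the same single wave player, so every quality level plays one ring recording.
# It plays at reduced volume with attenuation borrowed from the HauntedHouse exterior wind chimes prop, carries a send to the distant reverb submix, and has bus sends disabled.
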SoundCue /Game/Sounds/Artemis/POI/ACabin/Spots/Spot_ACabin_Clock_Ring_Cue
root=QualityLevel_0
nodes [
 {id=QualityLevel_0, children [WavePlayer_1,WavePlayer_1,WavePlayer_1,WavePlayer_1]}
 {id=WavePlayer_1, properties {soundWaveAssetPtr=&/Spot_ACabin_Clock_Ring_01}}
]
properties {
 attenuationSettings=/Game/Sounds/Artemis/POI/HauntedHouse/Spots/Spot_HauntedHouse_Exterior_WindChimes_Prop_Att
 bEnableBusSends=false
 maxDistance=1250
 soundClassObject=&Fort_Ambient_Outdoors_Athena
 volumeMultiplier=0.501
 preEffectBusSends [{}]
 soundSubmixSends [{
   minSendLevel=0.4
   sendLevel=0
   soundSubmix=&Fortnite_Reverb_Distant
   customSendLevelCurve {editorCurveData {keys [<at 0 0.4>,<at 1 1>]}}
  }]
}
