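# Sound cue for a cymbal rise one-shot spot ambience at the Artemis ACabin POI, classed as an indoor Athena ambient.
# A single wave player feeds a Modulator that randomizes each play, with a raised volume ceiling.
# The cue plays at low overall volume, references a closer tile-ambient one-shot attenuation asset, and sends to the distant reverb submix with a custom send curve.
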SoundCue /Game/Sounds/Artemis/POI/ACabin/Spots/Spot_ACabin_Cymbal_Rise_Cue
root=Modulator_0
nodes [
 {id=Modulator_0, children [WavePlayer_0], properties {volumeMax=1.2}}
 {id=WavePlayer_0, properties {soundWaveAssetPtr=&/Spot_ACabin_CymalRise_01}}
]
properties {
 attenuationSettings=/Game/Sounds/Attenuation/Fort_TileAmbient_OneShot_Att_Closer
 maxDistance=8500
 soundClassObject=&Fort_Ambient_Indoors_Athena
 volumeMultiplier=0.165
 attenuationOverrides {bAttenuate=false}
 soundSubmixSends [{
   maxSendLevel=1.4
   minSendLevel=0.6
   sendLevel=0.66
   soundSubmix=&Fortnite_Reverb_Distant
   customSendLevelCurve {editorCurveData {keys [<at 0 0.6>,<at 1 1>]}}
  }]
}
