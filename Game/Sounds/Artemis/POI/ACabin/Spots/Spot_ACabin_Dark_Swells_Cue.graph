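# Sound cue for dark swell ambient spot sounds at the Artemis ACabin POI, built from reused Fort_Starlight water impact swell waves.
# A Random node picks one of six equally weighted waves and feeds a Modulator left at default settings for pitch and volume variation.
# The cue boosts volume, references a closer one-shot tile ambient attenuation asset, and routes to an interior-space ducking ambient sound class.
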SoundCue /Game/Sounds/Artemis/POI/ACabin/Spots/Spot_ACabin_Dark_Swells_Cue
root=Modulator_0
nodes [
 {id=Modulator_0, children [Random_4]}
 {
  id=Random_4
  children [WavePlayer_19, WavePlayer_20, WavePlayer_21, WavePlayer_22, WavePlayer_23, WavePlayer_24]
  properties {weights [1,1,1,1,1,1]}
 }
 {id=WavePlayer_19, properties {soundWaveAssetPtr=/Game/Sounds/Fort_Starlight/S3/SoundWaves/Water_Impact_Swells_Random_01}}
 {id=WavePlayer_20, properties {soundWaveAssetPtr=/Game/Sounds/Fort_Starlight/S3/SoundWaves/Water_Impact_Swells_Random_02}}
 {id=WavePlayer_21, properties {soundWaveAssetPtr=/Game/Sounds/Fort_Starlight/S3/SoundWaves/Water_Impact_Swells_Random_03}}
 {id=WavePlayer_22, properties {soundWaveAssetPtr=/Game/Sounds/Fort_Starlight/S3/SoundWaves/Water_Impact_Swells_Random_04}}
 {id=WavePlayer_23, properties {soundWaveAssetPtr=/Game/Sounds/Fort_Starlight/S3/SoundWaves/Water_Impact_Swells_Random_05}}
 {id=WavePlayer_24, properties {soundWaveAssetPtr=/Game/Sounds/Fort_Starlight/S3/SoundWaves/Water_Impact_Swells_Random_06}}
]
properties {
 attenuationSettings=/Game/Sounds/Attenuation/Fort_TileAmbient_OneShot_Att_Closer
 maxDistance=8500
 soundClassObject=/Game/Sounds/FortSoundClassesAndModes/Fort_Ambient_InteriorSpace_Duck_TileMix_PSM
 volumeMultiplier=3
 attenuationOverrides {bAttenuate=false}
}
