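# Sound cue for the ACabin POI's demon-approaching spot sound, using an external attenuation asset and the Fort_Doors sound class.
# A Random node picks evenly between the same wave played plain or through an early-decaying volume envelope, with every quality level routed to it.
# A Modulator varies pitch and an outer envelope fades in, swells slightly, then fades out over 20 seconds.
# Concurrency override allows one instance and blocks new ones.
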SoundCue /Game/Sounds/Artemis/POI/ACabin/Spots/Spot_ACabin_DemonApproaching_Cue
root=Enveloper_0
nodes [
 {
  id=Enveloper_0
  children [Modulator_0]
  properties {
   pitchCurve {editorCurveData {keys [{time=0}]}}
   volumeCurve {editorCurveData {keys [
      <at 0 0>
      <at 1 1>
      <at 10 1.2>
      {arriveTangent=-0.0058, interpMode=RCIM_Cubic, leaveTangent=-0.0058, tangentMode=RCTM_User, time=12, value=1.2}
      {arriveTangent=-0.1926, interpMode=RCIM_Cubic, leaveTangent=-0.1926, tangentMode=RCTM_User, time=15, value=1}
      <at 20 0>
     ]}}
  }
 }
 {
  id=Enveloper_1
  children [WavePlayer_0]
  properties {
   pitchCurve {editorCurveData {keys [{time=0}]}}
   volumeCurve {editorCurveData {keys [
      <at 0 0>
      <at 1 1>
      <at 2 1>
      {arriveTangent=-0.0547, interpMode=RCIM_Cubic, leaveTangent=-0.0547, tangentMode=RCTM_User, time=8.3847, value=0.3166}
      <at 20 0>
     ]}}
  }
 }
 {id=Modulator_0, children [QualityLevel_0], properties {pitchMin=0.8}}
 {id=QualityLevel_0, children [Random_0,Random_0,Random_0,Random_0]}
 {id=Random_0, children [WavePlayer_7,Enveloper_1], properties {weights [1,1]}}
 {id=WavePlayer_0, properties {soundWaveAssetPtr=&/Spot_ACabin_DemonApproaching_01}}
 {id=WavePlayer_7, properties {soundWaveAssetPtr=&Spot_ACabin_DemonApproaching_01}}
]
properties {
 attenuationSettings=&/Spot_ACabin_DemonApproach_Att
 bOverrideConcurrency=true
 maxDistance=20000
 soundClassObject=&Fort_Doors
 volumeMultiplier=1.25
 attenuationOverrides {bAttenuate=false}
 concurrencyOverrides {maxCount=1, resolutionRule=PreventNew}
}
