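# Sound cue for a one-shot "demon uprise" ambient spot at the ACabin point of interest, interior variant.
# A Random node picks one of fourteen equally weighted wave variations, feeding every input of a Quality Level node, so all quality levels share the same pool.
# A root Modulator sits above that, and the cue uses the indoor ambient sound class with a wall-absorption monster-voice interior attenuation asset.
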SoundCue /Game/Sounds/Artemis/POI/ACabin/Spots/Spot_ACabin_DemonUprise_Int_Cue
root=Modulator_0
nodes [
 {id=Modulator_0, children [QualityLevel_0]}
 {id=QualityLevel_0, children [Random_0,Random_0,Random_0,Random_0]}
 {
  id=Random_0
  children [
   WavePlayer_0
   WavePlayer_1
   WavePlayer_2
   WavePlayer_3
   WavePlayer_4
   WavePlayer_5
   WavePlayer_6
   WavePlayer_10
   WavePlayer_11
   WavePlayer_12
   WavePlayer_13
   WavePlayer_14
   WavePlayer_15
   WavePlayer_16
  ]
  properties {weights [1,1,1,1,1,1,1,1,1,1,1,1,1,1]}
 }
 {id=WavePlayer_0, properties {soundWaveAssetPtr=&/Spot_ACabin_DemonUprise_14}}
 {id=WavePlayer_1, properties {soundWaveAssetPtr=&/Spot_ACabin_DemonUprise_01}}
 {id=WavePlayer_10, properties {soundWaveAssetPtr=&/Spot_ACabin_DemonUprise_07}}
 {id=WavePlayer_11, properties {soundWaveAssetPtr=&/Spot_ACabin_DemonUprise_08}}
 {id=WavePlayer_12, properties {soundWaveAssetPtr=&/Spot_ACabin_DemonUprise_09}}
 {id=WavePlayer_13, properties {soundWaveAssetPtr=&/Spot_ACabin_DemonUprise_10}}
 {id=WavePlayer_14, properties {soundWaveAssetPtr=&/Spot_ACabin_DemonUprise_11}}
 {id=WavePlayer_15, properties {soundWaveAssetPtr=&/Spot_ACabin_DemonUprise_12}}
 {id=WavePlayer_16, properties {soundWaveAssetPtr=&/Spot_ACabin_DemonUprise_13}}
 {id=WavePlayer_2, properties {soundWaveAssetPtr=&/Spot_ACabin_DemonUprise_02}}
 {id=WavePlayer_3, properties {soundWaveAssetPtr=&/Spot_ACabin_DemonUprise_03}}
 {id=WavePlayer_4, properties {soundWaveAssetPtr=&/Spot_ACabin_DemonUprise_04}}
 {id=WavePlayer_5, properties {soundWaveAssetPtr=&/Spot_ACabin_DemonUprise_05}}
 {id=WavePlayer_6, properties {soundWaveAssetPtr=&/Spot_ACabin_DemonUprise_06}}
]
properties {
 attenuationSettings=&/Fort_TileAmbient_OneShot_Monster_Voice_WallAbsorbtion_Int_Att
 maxDistance=8000
 soundClassObject=&Fort_Ambient_Indoors_Athena
 volumeMultiplier=1.002
 attenuationOverrides {bAttenuate=false}
}
