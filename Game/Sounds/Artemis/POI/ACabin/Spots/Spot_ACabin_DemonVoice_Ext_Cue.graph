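# Sound cue for an exterior demon-voice ambient spot sound at the ACabin point of interest, classed as outdoor ambience.
# A random node picks one of three voice waves with equal weight, boosting the third through a mixer.
# All quality levels share that random node, and a root modulator wraps the result.
# It uses a closer tile-ambient one-shot attenuation asset and sends to the distant reverb submix with a custom level curve.
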SoundCue /Game/Sounds/Artemis/POI/ACabin/Spots/Spot_ACabin_DemonVoice_Ext_Cue
root=Modulator_0
nodes [
 {id=Mixer_0, children [WavePlayer_9], properties {inputVolume [1.25]}}
 {id=Modulator_0, children [QualityLevel_0]}
 {id=QualityLevel_0, children [Random_1,Random_1,Random_1,Random_1]}
 {id=Random_1, children [WavePlayer_7,WavePlayer_8,Mixer_0], properties {weights [1,1,1]}}
 {id=WavePlayer_7, properties {soundWaveAssetPtr=&/Spot_ACabin_DemonVoice_01}}
 {id=WavePlayer_8, properties {soundWaveAssetPtr=&/Spot_ACabin_DemonVoice_02}}
 {id=WavePlayer_9, properties {soundWaveAssetPtr=&/Spot_ACabin_DemonVoice_03}}
]
properties {
 attenuationSettings=/Game/Sounds/Attenuation/Fort_TileAmbient_OneShot_Att_Closer
 maxDistance=8500
 soundClassObject=&Fort_Ambient_Outdoors_Athena
 volumeMultiplier=0.601
 attenuationOverrides {bAttenuate=false}
 soundSubmixSends [{
   minSendLevel=0.4
   sendLevel=0
   soundSubmix=&Fortnite_Reverb_Distant
   customSendLevelCurve {editorCurveData {keys [<at 0 0.25>,<at 1 1>]}}
  }]
}
