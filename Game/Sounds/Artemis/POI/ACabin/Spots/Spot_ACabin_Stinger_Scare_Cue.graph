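# Sound cue for a scare stinger spot sound at the ACabin POI, playing a single wave on the Fort_Doors sound class.
# A modulator randomizes pitch and volume over a quality-level node whose four branches all point to the same wave player.
# It references a shared interior, wall-absorption monster-voice attenuation asset, boosts overall volume, and adds a send to the distant reverb submix.
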SoundCue /Game/Sounds/Artemis/POI/ACabin/Spots/Spot_ACabin_Stinger_Scare_Cue
root=Modulator_0
nodes [
 {id=Modulator_0, children [QualityLevel_0], properties {pitchMin=0.7, volumeMin=0.7}}
 {id=QualityLevel_0, children [WavePlayer_0,WavePlayer_0,WavePlayer_0,WavePlayer_0]}
 {id=WavePlayer_0, properties {soundWaveAssetPtr=&/Spot_ACabin_Stinger_Scare_01}}
]
properties {
 attenuationSettings=&/Fort_TileAmbient_OneShot_Monster_Voice_WallAbsorbtion_Int_Att
 maxDistance=8000
 soundClassObject=&Fort_Doors
 volumeMultiplier=1.5
 attenuationOverrides {bAttenuate=false}
 soundSubmixSends [{
   maxSendLevel=1.2
   minSendLevel=0.4
   sendLevel=0
   soundSubmix=&Fortnite_Reverb_Distant
   customSendLevelCurve {editorCurveData {keys [<at 0 0.4>,<at 1 1>]}}
  }]
}
