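# Sound Cue for a vocal scare spot effect at the ACabin point of interest.
# A random node picks one of two vocal waves with equal weight, with an Enveloper fading the first to silence over about two seconds.
# All quality levels share that branch, which feeds a Modulator at the root.
# It uses a distant interior attenuation asset, the Fort_Doors sound class, boosted volume and a distant reverb submix send.
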SoundCue /Game/Sounds/Artemis/POI/ACabin/Spots/Spot_ACabin_VocalScare_Cue
root=Modulator_0
nodes [
 {
  id=Enveloper_0
  children [WavePlayer_0]
  properties {
   pitchCurve {editorCurveData {keys [{time=0}]}}
   volumeCurve {editorCurveData {keys [<at 0 1>, {arriveTangent=-0.5739, interpMode=RCIM_Cubic, leaveTangent=-0.5739, time=1.1932, value=0.8568}, <at 1.7556 -0.0076 RCIM_Cubic>]}}
  }
 }
 {id=Modulator_0, children [QualityLevel_0]}
 {id=QualityLevel_0, children [Random_0,Random_0,Random_0,Random_0]}
 {id=Random_0, children [Enveloper_0,WavePlayer_1], properties {weights [1,1]}}
 {id=WavePlayer_0, properties {soundWaveAssetPtr=&/Spot_ACabin_VocalScare_01}}
 {id=WavePlayer_1, properties {soundWaveAssetPtr=&/Spot_ACabin_VocalScare_02}}
]
properties {
 attenuationSettings=&/Spot_ACabin_Vocal_Scare_Int_Distant_Att
 maxDistance=8000
 soundClassObject=&Fort_Doors
 volumeMultiplier=4
 attenuationOverrides {bAttenuate=false}
 soundSubmixSends [{
   minSendLevel=0.4
   sendLevel=0
   soundSubmix=&Fortnite_Reverb_Distant
   customSendLevelCurve {editorCurveData {keys [<at 0 0.4>,<at 1 1>]}}
  }]
}
