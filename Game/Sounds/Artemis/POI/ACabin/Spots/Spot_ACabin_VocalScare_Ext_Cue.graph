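# Sound cue for an exterior ambient vocal scare spot at the ACabin point of interest, using the outdoor ambient sound class.
# It picks one of two equally weighted vocal scare waves, with every quality level routed to that same random node, then varies it through a modulator with a lowered minimum pitch.
# It borrows the interior distant witch laughs attenuation asset and sends to a distant reverb submix.
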SoundCue /Game/Sounds/Artemis/POI/ACabin/Spots/Spot_ACabin_VocalScare_Ext_Cue
root=Modulator_0
nodes [
 {id=Modulator_0, children [QualityLevel_0], properties {pitchMin=0.8}}
 {id=QualityLevel_0, children [Random_0,Random_0,Random_0,Random_0]}
 {id=Random_0, children [WavePlayer_0,WavePlayer_1], properties {weights [1,1]}}
 {id=WavePlayer_0, properties {soundWaveAssetPtr=&/Spot_ACabin_VocalScare_01}}
 {id=WavePlayer_1, properties {soundWaveAssetPtr=&/Spot_ACabin_VocalScare_02}}
]
properties {
 attenuationSettings=&/Spot_ACabin_Witch_Laughs_Int_Distant_Att
 maxDistance=8000
 soundClassObject=&Fort_Ambient_Outdoors_Athena
 volumeMultiplier=2.506
 attenuationOverrides {bAttenuate=false}
 soundSubmixSends [{
   minSendLevel=0.6
   sendLevel=0.6
   soundSubmix=&Fortnite_Reverb_Distant
   customSendLevelCurve {editorCurveData {keys [<at 0 0.6>,<at 1 1>]}}
  }]
}
