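# Exterior ambient one-shot cue for witch laughter spots at the ACabin POI, classed as outdoor ambience.
# A root Modulator feeds a Quality Level node whose four tiers share one equally weighted Random node with sixteen laugh variations.
# Only laugh 15 passes through a Mixer, which doubles its volume.
# The cue references a shared wide one-shot attenuation asset and defines a send to the distant reverb submix.
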SoundCue /Game/Sounds/Artemis/POI/ACabin/Spots/Spot_ACabin_Witch_Laughs_Ext_Cue
root=Modulator_0
nodes [
 {id=Mixer_1, children [WavePlayer_15], properties {inputVolume [2]}}
 {id=Modulator_0, children [QualityLevel_0]}
 {id=QualityLevel_0, children [Random_0,Random_0,Random_0,Random_0]}
 {
  id=Random_0
  children [
   WavePlayer_0
   WavePlayer_1
   WavePlayer_2
   WavePlayer_3
   WavePlayer_4
   WavePlayer_5
   WavePlayer_6
   WavePlayer_10
   WavePlayer_11
   WavePlayer_12
   WavePlayer_13
   WavePlayer_14
   Mixer_1
   WavePlayer_21
   WavePlayer_22
   WavePlayer_23
  ]
  properties {weights [1,1,1,1,1,1,1,1,1,1,1,1,1,1,1,1]}
 }
 {id=WavePlayer_0, properties {soundWaveAssetPtr=&/Spot_ACabin_Witches_Laughing_01}}
 {id=WavePlayer_1, properties {soundWaveAssetPtr=&/Spot_ACabin_Witches_Laughing_02}}
 {id=WavePlayer_10, properties {soundWaveAssetPtr=&/Spot_ACabin_Witches_Laughing_08}}
 {id=WavePlayer_11, properties {soundWaveAssetPtr=&/Spot_ACabin_Witches_Laughing_09}}
 {id=WavePlayer_12, properties {soundWaveAssetPtr=&/Spot_ACabin_Witches_Laughing_10}}
 {id=WavePlayer_13, properties {soundWaveAssetPtr=&/Spot_ACabin_Witches_Laughing_11}}
 {id=WavePlayer_14, properties {soundWaveAssetPtr=&/Spot_ACabin_Witches_Laughing_13}}
 {id=WavePlayer_15, properties {soundWaveAssetPtr=&/Spot_ACabin_Witches_Laughing_15}}
 {id=WavePlayer_2, properties {soundWaveAssetPtr=&/Spot_ACabin_Witches_Laughing_03}}
 {id=WavePlayer_21, properties {soundWaveAssetPtr=&/Spot_ACabin_Witches_Laughing_21}}
 {id=WavePlayer_22, properties {soundWaveAssetPtr=&/Spot_ACabin_Witches_Laughing_22}}
 {id=WavePlayer_23, properties {soundWaveAssetPtr=&/Spot_ACabin_Witches_Laughing_23}}
 {id=WavePlayer_3, properties {soundWaveAssetPtr=&/Spot_ACabin_Witches_Laughing_04}}
 {id=WavePlayer_4, properties {soundWaveAssetPtr=&/Spot_ACabin_Witches_Laughing_05}}
 {id=WavePlayer_5, properties {soundWaveAssetPtr=&/Spot_ACabin_Witches_Laughing_06}}
 {id=WavePlayer_6, properties {soundWaveAssetPtr=&/Spot_ACabin_Witches_Laughing_07}}
]
properties {
 attenuationSettings=/Game/Sounds/Attenuation/Fort_TileAmbient_OneShot_Wider_Att
 maxDistance=8000
 soundClassObject=&Fort_Ambient_Outdoors_Athena
 volumeMultiplier=1.002
 attenuationOverrides {bAttenuate=false}
 soundSubmixSends [{
   minSendLevel=0.4
   sendLevel=0
   soundSubmix=&Fortnite_Reverb_Distant
   customSendLevelCurve {editorCurveData {keys [<at 0 0.4>,<at 1 1>]}}
  }]
}
