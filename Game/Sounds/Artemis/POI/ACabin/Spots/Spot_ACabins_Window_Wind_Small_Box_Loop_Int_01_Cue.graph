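# SoundCue for a looping indoor wind-through-window ambience spot at the ACabin POI, reusing a Hollow_Base mountain wind wave.
# Nested looping Envelopers vary pitch over 30 seconds and volume over 60 seconds, beneath an outer two-second fade-in.
# A quality node feeds that chain to its first two levels and leaves the other two empty.
# Small box attenuation and a two-voice concurrency limit keep it localized.
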SoundCue /Game/Sounds/Artemis/POI/ACabin/Spots/Spot_ACabins_Window_Wind_Small_Box_Loop_Int_01_Cue
root=Enveloper_1
nodes [
 {
  id=Enveloper_0
  children [WavePlayer_0]
  properties {
   bLoop=true
   bLoopIndefinitely=true
   loopEnd=30
   pitchCurve {editorCurveData {keys [{interpMode=RCIM_Cubic, time=0}, <at 15 0.9 RCIM_Cubic>, {interpMode=RCIM_Cubic, time=30}]}}
   volumeCurve {editorCurveData {keys [<at 0 1>]}}
  }
 }
 {
  id=Enveloper_1
  children [QualityLevel_0]
  properties {pitchCurve {editorCurveData {keys [{time=0}]}}, volumeCurve {editorCurveData {keys [<at 0 0 RCIM_Cubic>,<at 2 1 RCIM_Cubic>]}}}
 }
 {
  id=Enveloper_3
  children [Enveloper_0]
  properties {
   bLoop=true
   bLoopIndefinitely=true
   loopEnd=60
   pitchCurve {editorCurveData {keys [{time=0}]}}
   volumeCurve {editorCurveData {keys [
      {arriveTangent=-0.01, interpMode=RCIM_Cubic, time=0, value=1}
      <at 10 0.8 RCIM_Cubic>
      {arriveTangent=0.0095, interpMode=RCIM_Cubic, leaveTangent=0.0095, time=20, value=1}
      <at 31.1629 1>
      {arriveTangent=-0.0016, interpMode=RCIM_Cubic, leaveTangent=-0.0016, time=37.937, value=0.5372}
      <at 46.7125 0.975>
      <at 60 1 RCIM_Cubic>
     ]}}
  }
 }
 {id=QualityLevel_0, children [Enveloper_3,Enveloper_3,null,null]}
 {
  id=WavePlayer_0
  properties {bLooping=true, soundWaveAssetPtr=/Game/Sounds/Artemis/POI/Hollow_Base/Spots/Spot_Mountain_Wind_Window_Loop_Stereo}
 }
]
properties {
 bOverrideAttenuation=true
 bOverrideConcurrency=true
 maxDistance=400
 soundClassObject=&Fort_Ambient_Indoors_Athena
 volumeMultiplier=1.002
 attenuationOverrides {attenuationShape=Box, falloffDistance=300, attenuationShapeExtents=<90 82.6418 100>}
 concurrencyOverrides {maxCount=2, voiceStealReleaseTime=2}
}
