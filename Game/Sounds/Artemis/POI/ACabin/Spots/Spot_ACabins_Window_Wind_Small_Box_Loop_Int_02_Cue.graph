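# Looping indoor ambience cue for wind at a small cabin window, reusing a stereo desert wind loop on the indoor ambient sound class.
# Nested looping envelopes drift pitch and swell volume over different periods, varying the repeated wave, under a short fade-in.
# A quality node feeds that chain to its first two slots and leaves the other two empty.
# Overrides confine it to a small box attenuation and a single concurrent voice.
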SoundCue /Game/Sounds/Artemis/POI/ACabin/Spots/Spot_ACabins_Window_Wind_Small_Box_Loop_Int_02_Cue
root=Enveloper_1
nodes [
 {
  id=Enveloper_1
  children [QualityLevel_0]
  properties {pitchCurve {editorCurveData {keys [{time=0}]}}, volumeCurve {editorCurveData {keys [<at 0 0 RCIM_Cubic>,<at 2 1 RCIM_Cubic>]}}}
 }
 {
  id=Enveloper_2
  children [WavePlayer_0]
  properties {
   bLoop=true
   bLoopIndefinitely=true
   loopEnd=30
   pitchCurve {editorCurveData {keys [{interpMode=RCIM_Cubic, time=0}, <at 15.9468 0.9181 RCIM_Cubic>, {interpMode=RCIM_Cubic, time=30}]}}
   volumeCurve {editorCurveData {keys [<at 0 1>]}}
  }
 }
 {
  id=Enveloper_4
  children [Enveloper_2]
  properties {
   bLoop=true
   bLoopIndefinitely=true
   loopEnd=20
   pitchCurve {editorCurveData {keys [{time=0}]}}
   volumeCurve {editorCurveData {keys [
      {arriveTangent=-0.01, interpMode=RCIM_Cubic, time=0, value=1}
      <at 10 1.2 RCIM_Cubic>
      {interpMode=RCIM_Cubic, leaveTangent=0.01, time=20, value=1}
     ]}}
  }
 }
 {id=QualityLevel_0, children [Enveloper_4,Enveloper_4,null,null]}
 {id=WavePlayer_0, properties {bLooping=true, soundWaveAssetPtr=/Game/Sounds/Artemis/POI/Seven_Workshop/Spot_Desert_Wind_Window_Loop_Stereo}}
]
properties {
 bOverrideAttenuation=true
 bOverrideConcurrency=true
 maxDistance=400
 soundClassObject=&Fort_Ambient_Indoors_Athena
 volumeMultiplier=1.002
 attenuationOverrides {attenuationShape=Box, falloffDistance=300, attenuationShapeExtents=<90 100 50>}
 concurrencyOverrides {maxCount=1, voiceStealReleaseTime=2}
}
