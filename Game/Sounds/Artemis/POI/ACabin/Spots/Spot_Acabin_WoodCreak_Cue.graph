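# Sound cue for spot ambient wood creaks at the ACabin point of interest, classed as indoor ambience.
# A random node picks among nine equally weighted creak waves and feeds every input of a quality-level node, so all quality tiers share the same pool.
# A modulator at the root varies pitch, and the cue plays at reduced volume with a distant-creak one-shot attenuation asset.
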SoundCue /Game/Sounds/Artemis/POI/ACabin/Spots/Spot_Acabin_WoodCreak_Cue
root=Modulator_1
nodes [
 {id=Modulator_1, children [QualityLevel_0], properties {pitchMin=0.8}}
 {id=QualityLevel_0, children [Random_0,Random_0,Random_0,Random_0]}
 {
  id=Random_0
  children [WavePlayer_0,WavePlayer_1,WavePlayer_2,WavePlayer_3,WavePlayer_4,WavePlayer_5,WavePlayer_6,WavePlayer_7,WavePlayer_8]
  properties {weights [1,1,1,1,1,1,1,1,1]}
 }
 {id=WavePlayer_0, properties {soundWaveAssetPtr=&/Spot_ACabin_WoodCreak_01}}
 {id=WavePlayer_1, properties {soundWaveAssetPtr=&/Spot_ACabin_WoodCreak_02}}
 {id=WavePlayer_2, properties {soundWaveAssetPtr=&/Spot_ACabin_WoodCreak_03}}
 {id=WavePlayer_3, properties {soundWaveAssetPtr=&/Spot_ACabin_WoodCreak_04}}
 {id=WavePlayer_4, properties {soundWaveAssetPtr=&/Spot_ACabin_WoodCreak_05}}
 {id=WavePlayer_5, properties {soundWaveAssetPtr=&/Spot_ACabin_WoodCreak_06}}
 {id=WavePlayer_6, properties {soundWaveAssetPtr=&/Spot_ACabin_WoodCreak_07}}
 {id=WavePlayer_7, properties {soundWaveAssetPtr=&/Spot_ACabin_WoodCreak_08}}
 {id=WavePlayer_8, properties {soundWaveAssetPtr=&/Spot_ACabin_WoodCreak_09}}
]
properties {
 attenuationSettings=/Game/Sounds/Attenuation/Fort_TileAmbient_OneShot_Distant_Creak_Att
 maxDistance=8200
 soundClassObject=&Fort_Ambient_Indoors_Athena
 volumeMultiplier=0.501
 attenuationOverrides {bAttenuate=false}
}
